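; Time of Day Manager for BR Artemis that adds a temporary, console-toggled audio render-target analysis feature for perf tests.
; At BeginPlay it binds to data-driven CVar changes and checks RT.AudioAnalysis, spawning a soft-referenced audio actor when true, otherwise expiring it.
; Helper functions clear two render targets and draw sample and propagate materials into them off dedicated servers, but nothing here calls them.
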(event Custom|ValidateRTConsoleVars
  (params (ConsoleVar :type String))
  ;[1 1 1] RT Audio Analysis (TEMPORARY - Will be removed after Perf Tests )
  (bind _analyseaudiort (CallFunction|ValidateRTConsoleVariable ConsoleVar))
  (Variables|Default|SetBAnalyse%20Audio%20RT _analyseaudiort)
  (if _analyseaudiort
    (Development|PrintString "SpawnRTBP")
    (bind _returnvalue (Utilities|LoadClassAssetBlocking (Variables|Default|GetAudioSoftRef)))
    (bind _asactor (Utilities|Casting|CastToActorClass _returnvalue))
    (bind _returnvalue_2 (Game|SpawnActorfromClass _asactor (Variables|Default|GetSpawn%20Transform) "AlwaysSpawn" "SelectDefaultAtRuntime" 0 0))
    (Variables|Default|SetAudioBP _returnvalue_2)
    (else
      (Development|PrintString "RemoveRTBP" true true "(R=1.000000,G=0.000000,B=0.000000,A=1.000000)")
      (Actor|SetLifeSpan (Variables|Default|GetAudioBP) 0.1)
      (Variables|Default|SetAudioBP 0))))

(event EventBeginPlay
  (|CallParentFunction)
  (CallFunction|InitializeConsoleVars))

(event EventTick
  (params (DeltaSeconds :type Float%20%28single-precision%29))
  (|CallParentFunction DeltaSeconds))

(fn InitializeAudioRenderTarget
  (bind _returnvalue (Rendering|Material|CreateDynamicMaterialInstance "/Game/ArtTools/Tech_Art/AudioRT/M_SampleAudio_MPC.M_SampleAudio_MPC" "None" "Transient" :!class /Script/Engine.KismetMaterialLibrary))
  (Variables|Default|SetM_SampleAudioMID _returnvalue)
  (Rendering|ClearRenderTarget2D "/Game/ArtTools/Tech_Art/AudioRT/RT_Audio_01.RT_Audio_01")
  (Rendering|ClearRenderTarget2D "/Game/ArtTools/Tech_Art/AudioRT/RT_Audio_02.RT_Audio_02"))

(fn InitializeConsoleVars
  (DataDrivenCVar|BindEventtoOnDataDrivenCVarDelegate (EngineSubsystems|GetDataDrivenCVarEngineSubsystem) (EventDispatchers|CreateEvent :!event ValidateRTConsoleVars))
  (CallFunction|ValidateRTConsoleVars "RT.AudioAnalysis"))

(fn UpdateAudioRT
  (if (not (Networking|IsDedicatedServer))
    (Rendering|DrawMaterialtoRenderTarget "/Game/ArtTools/Tech_Art/AudioRT/RT_Audio_01.RT_Audio_01" (Variables|Default|GetM_SampleAudioMID))
    (Rendering|DrawMaterialtoRenderTarget "/Game/ArtTools/Tech_Art/AudioRT/RT_Audio_02.RT_Audio_02" "/Game/ArtTools/Tech_Art/AudioRT/M_AudioPropogate.M_AudioPropogate")))

(fn ConstructionScript
  (|CallParentFunction))

(fn ValidateRTConsoleVariable
  (params (ConsoleVAr :type String))
  (if (Utilities|String|EqualExactly(String) ConsoleVAr "RT.AudioAnalysis")
    (bind _returnvalue (Development|GetConsoleVariableBoolValue "RT.AudioAnalysis"))
    (return _returnvalue)))
