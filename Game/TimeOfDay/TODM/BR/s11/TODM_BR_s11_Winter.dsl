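; Winter time-of-day Blueprint for BR s11 that drives snowstorm visuals and audio from weather events.
; Weather start shows a camera sphere with a dynamic material, passes it the event index, and spawns start and loop sounds.
; A repeating timer feeds weather intensity, event index and blend into the loop's audio parameters.
; Weather end plays a stop sound, fades the loop, and hides the sphere after a delay.
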(event EventBeginPlay
  (|CallParentFunction) ; This node is disabled and will not be called.\nDrag off pins to build functionality.
)

(event EventTick
  (params (DeltaSeconds :type Float%20%28single-precision%29))
  (|CallParentFunction DeltaSeconds) ; This node is disabled and will not be called.\nDrag off pins to build functionality.
)

(event EventOnWeatherStart
  (params (WeatherEventIndex :type Integer))
  (bind _snowstormcamerasphere (Variables|Default|GetSnowStormCameraSphere))
  ;[1 1 1] Snow Sphere Toggling only when weather is active medium, high or extreme
  (Rendering|SetVisibility (Variables|Default|GetSnowStormCameraSphere) true false)
  (Rendering|Material|SetScalarParameterValueonMaterials _snowstormcamerasphere "WeatherIndexFromBP" (Math|Conversions|ToFloat(Integer) (GlobalWeather|GetCurrentWeatherEventIndex (Variables|Default|GetFortTimeOfDayWeather))))
  (Rendering|Material|SetScalarParameterValueonMaterials _snowstormcamerasphere "ActivateCameraPos_FromBP" 1.0)
  ;[1 1 1] Audio Events
  (CallFunction|StartSnowStormAudio))

(event EventOnWeatherEnd
  (params (WeatherEventIndex :type Integer))
  (bind _snowstormcamerasphere (Variables|Default|GetSnowStormCameraSphere))
  ;[1 1 1] Audio Events
  (CallFunction|StopSnowStormAudio)
  (Utilities|FlowControl|Delay 2e+01)
  (Rendering|SetVisibility _snowstormcamerasphere false false)
  (Rendering|Material|SetScalarParameterValueonMaterials _snowstormcamerasphere "ActivateCameraPos_FromBP" 0.0))

(event Custom|Start%20Snow%20Storm%20Audio
  (bind _returnvalue (Math|Vector|MakeVector -1.0))
  (bind _returnvalue_2 (* _returnvalue 1e+06))
  (Variables|Default|GetSnowStormAudioComponent
    (:then)
    (:else
      (Utilities|FlowControl|Sequence
        (:then_0
          (bind _returnvalue_3 (Audio|SpawnSoundatLocation (Variables|Sound|GetSound_SnowStormLoop) _returnvalue_2))
          (bind _returnvalue_4 (Audio|SpawnSoundatLocation (Variables|Sound|GetSound_SnowStormStart) _returnvalue_2))
          (bind _output_get (Variables|Default|SetSnowStormAudioComponent _returnvalue_3))
          (CallFunction|UpdateAudiofromStormParams _output_get)
          (CallFunction|UpdateAudiofromStormParams _returnvalue_4))
        (:then_1
          (Utilities|Time|SetTimerbyEvent (EventDispatchers|CreateEvent :!event Custom|On%20Snow%20Storm%20Audio%20Update) 0.2 true)
          (CallFunction|OnSnowStormAudioUpdate))))))

(event Custom|Stop%20Snow%20Storm%20Audio
  (Variables|Default|GetSnowStormAudioComponent
    (:then
      (Utilities|FlowControl|Sequence
        (:then_0
          (Utilities|Time|SetTimerbyEvent (EventDispatchers|CreateEvent :!event Custom|On%20Snow%20Storm%20Audio%20Update)))
        (:then_1
          (CallFunction|UpdateAudiofromStormParams (Variables|Default|GetSnowStormAudioComponent))
          (bind _returnvalue (Audio|SpawnSoundatLocation (Variables|Sound|GetSound_SnowStormStop) (* (Math|Vector|MakeVector -1.0) 1e+06)))
          (CallFunction|UpdateAudiofromStormParams _returnvalue)
          (Audio|Components|Audio|FadeOut (Variables|Default|GetSnowStormAudioComponent) 1e+01 0.0 "Linear" :!class /Script/Engine.AudioComponent))))
    (:else)))

(event Custom|On%20Snow%20Storm%20Audio%20Update
  (CallFunction|UpdateAudiofromStormParams (Variables|Default|GetSnowStormAudioComponent)))

(fn UpdateAudioFromStormParams
  (params (AudioComponent :type Audio%20Component%20Object%20Reference))
  (bind _forttimeofdayweather (Variables|Default|GetFortTimeOfDayWeather))
  (Audio|Parameter|SetFloatParameter AudioComponent "Intensity" (Class|FortTimeOfDayWeatherComponent|GetCurrentWeatherIntensity _forttimeofdayweather) :!class /Script/AudioExtensions.AudioParameterControllerInterface)
  (Audio|Parameter|SetIntegerParameter AudioComponent "EventIndex" (GlobalWeather|GetCurrentWeatherEventIndex _forttimeofdayweather) :!class /Script/AudioExtensions.AudioParameterControllerInterface)
  (Audio|Parameter|SetFloatParameter AudioComponent "Blend" (GlobalWeather|GetBlendTimePercentage _forttimeofdayweather) :!class /Script/AudioExtensions.AudioParameterControllerInterface))

(fn ConstructionScript
  (|CallParentFunction)
  (Rendering|Material|CreateDynamicMaterialInstance (Variables|Default|GetSnowStormCameraSphere) 0 0 "SphereMID" :!class /Script/Engine.PrimitiveComponent))
