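; Battle Royale time-of-day manager Blueprint (TODM_BR_s14) that gates advanced post-processing and exposes sky and storm visual hooks.
; ValidateAdvancedPostProcessing enables the JanusPost component from the Athena.PostProcessLevel CVar, inside an AllowJanus macro.
; It re-runs on editor construction, BeginPlay and CVar changes.
; Custom events hide the sky dome for Cyclone, toggle storm visuals through the MPC_SafeZoneClouds parameter collection, and provide temporary JanusPost test switches.
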(event EventOnDataDrivenCVarChangedEditorOnly
  (params (CVarName :type String))
  (if (Utilities|String|EqualExactly(String) CVarName "Athena.PostProcessLevel")
    (CallFunction|ValidateAdvancedPostProcessing)))

(event Custom|CycloneTODM
  ;[1 1 1] This is used by the Cyclone event
  (Rendering|SetVisibility (Variables|DayNightCycle|GetSkyDomeMeshComp) false false))

(event EventBeginPlay
  (|CallParentFunction)
  (CallFunction|ValidateAdvancedPostProcessing)
  (DataDrivenCVar|AssignOnDataDrivenCVarDelegate (EngineSubsystems|GetDataDrivenCVarEngineSubsystem) (delegate Custom|OnDataDrivenCVarDelegate_Event_0)))

(event Custom|OnDataDrivenCVarDelegate_Event_0
  (params (CVarName :type String))
  (if (Utilities|String|EqualExactly(String) CVarName "Athena.PostProcessLevel")
    (CallFunction|ValidateAdvancedPostProcessing)))

(event Custom|JanusPostEnable
  ;[1 1 1] Kill these after testing CVAR
  (Variables|/Script/Engine.PostProcessComponent|SetbEnabled true (Variables|Default|GetJanusPost)))

(event Custom|JanusPostDisable
  ;[1 1 1] Kill these after testing CVAR
  (Variables|/Script/Engine.PostProcessComponent|SetbEnabled false (Variables|Default|GetJanusPost)))

(event Custom|RollStormOut
  ;[1 1 1] Disable/Enable Storm Visual
  (Networking|FlushNetDormancy)
  (Rendering|Material|SetScalarParameterValue "/Game/Athena/SafeZone/MPC_SafeZoneClouds.MPC_SafeZoneClouds" "PreventStormShrinking" 1.0 :!class /Script/Engine.KismetMaterialLibrary))

(event Custom|RollStormIn
  ;[1 1 1] Disable/Enable Storm Visual
  (Networking|FlushNetDormancy)
  (Rendering|Material|SetScalarParameterValue "/Game/Athena/SafeZone/MPC_SafeZoneClouds.MPC_SafeZoneClouds" "PreventStormShrinking" :!class /Script/Engine.KismetMaterialLibrary))

(fn ConstructionScript
  (|CallParentFunction)
  (if (Editor|InEditororPIE)
    (CallFunction|ValidateAdvancedPostProcessing)))

(fn validateAdvancedPostProcessing
  (Utilities|AllowJanus
    (:Allow
      (bind _returnvalue (Development|GetConsoleVariableIntValue "Athena.PostProcessLevel"))
      (Variables|/Script/Engine.PostProcessComponent|SetbEnabled (>= _returnvalue 3) (Variables|Default|GetJanusPost)))))
