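; TODM_Creative is a time-of-day manager Blueprint, apparently for Creative, with an event that hides the sky dome mesh for the Cyclone event.
; It calls ValidateAdvancedPostProcessing at BeginPlay, during construction in editor or PIE, and whenever the Athena.PostProcessLevel CVar changes.
; BeginPlay binds the runtime CVar delegate, while a separate handler covers editor-only changes.
; The validation function and both JanusPost enable/disable events are empty in this file.
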(event Custom|CycloneTODM
  ;[1 1 1] This is used by the Cyclone event
  (Rendering|SetVisibility (Variables|DayNightCycle|GetSkyDomeMeshComp) false false))

(event EventOnDataDrivenCVarChangedEditorOnly
  (params (CVarName :type String))
  (if (Utilities|String|EqualExactly(String) CVarName "Athena.PostProcessLevel")
    (CallFunction|ValidateAdvancedPostProcessing)))

(event EventBeginPlay
  (|CallParentFunction)
  (CallFunction|ValidateAdvancedPostProcessing)
  (DataDrivenCVar|AssignOnDataDrivenCVarDelegate (EngineSubsystems|GetDataDrivenCVarEngineSubsystem) (delegate Custom|OnDataDrivenCVarDelegate_Event_0)))

(event Custom|OnDataDrivenCVarDelegate_Event_0
  (params (CVarName :type String))
  (if (Utilities|String|EqualExactly(String) CVarName "Athena.PostProcessLevel")
    (CallFunction|ValidateAdvancedPostProcessing)))

(event Custom|JanusPostEnable)

(event Custom|JanusPostDisable)

(fn ConstructionScript
  (|CallParentFunction)
  (if (Editor|InEditororPIE)
    (CallFunction|ValidateAdvancedPostProcessing)))

(fn validateAdvancedPostProcessing)
